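; Player pawn Blueprint for the Lil'Whip NPC that caps how many thrown ice cream pickups exist at once.
; IceCreamTracker_AddPickUp runs on authority, destroying the oldest tracked pickup when the list reaches a scalable-float limit, then registering the new one.
; Each tracked pickup binds picked-up and destroyed delegates, which unbind themselves and remove it from the list to keep the count accurate.
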(event EventBeginPlay
  (|CallParentFunction) ; This node is disabled and will not be called.\nDrag off pins to build functionality.
)

(event Collision|EventActorBeginOverlap
  (params (OtherActor :type Actor%20Object%20Reference))
  (|CallParentFunction OtherActor) ; This node is disabled and will not be called.\nDrag off pins to build functionality.
)

(event EventTick
  (params (DeltaSeconds :type Float%20%28single-precision%29))
  (|CallParentFunction DeltaSeconds) ; This node is disabled and will not be called.\nDrag off pins to build functionality.
)

(event Custom|IceCreamTracker_AddPickUp
  (params (NewIceCreamPickUp :type Fort%20Pickup%20Object%20Reference))
  (bind _thrownicecreams (Variables|Default|GetThrownIceCreams))
  ;[1 1 1] Logic to keep track of how many icecreams Lil'Whip has thrown and make sure we dont have too many.
  (Networking|SwitchHasAuthority
    (:Authority
      (Utilities|IsValid NewIceCreamPickUp
        (:Is%20Valid
          (bind _output_get (Variables|Default|SetNewPIckUp NewIceCreamPickUp))
          (if (>= (Utilities|Array|Length (Variables|Default|GetThrownIceCreams)) (ScalableFloat|GetValueatLevel (Variables|Default|GetMaxIceCreamPickUps)))
            (bind _output_get_2 (Variables|Default|SetPickUpToDestroy (Utilities|Array|Get(acopy) _thrownicecreams)))
            (Utilities|Array|RemoveIndex _thrownicecreams)
            (Actor|DestroyActor _output_get_2)
            (CallFunction|AddPickUptoList (Variables|Default|GetNewPIckUp))
            (else
              (CallFunction|AddPickUptoList _output_get))))))))

(event Custom|AddPickUpToList
  (params (NewPickUp :type Fort%20Pickup%20Object%20Reference))
  ;[1 1 1] Keep Track if the Item was PickedUp and or Destroyed to make sure the List is up to date.
  (Utilities|Array|AddUnique (Variables|Default|GetThrownIceCreams) NewPickUp)
  (Interaction|BindEventtoOnPickup NewPickUp (delegate Custom|IceCreamWasPickedUp))
  (Game|BindEventtoOnDestroyed NewPickUp (delegate Custom|PickUpWasDetroyed)))

(event Custom|IceCreamWasPickedUp
  (params (SelfActor :type Fort%20Pickup%20Object%20Reference) (InteractingPawn :type Fort%20Pawn%20Object%20Reference))
  ;[1 1 1] Keep Track if the Item was PickedUp and or Destroyed to make sure the List is up to date.
  (Interaction|UnbindEventfromOnPickup SelfActor (delegate Custom|IceCreamWasPickedUp))
  (Game|UnbindEventfromOnDestroyed SelfActor (delegate Custom|PickUpWasDetroyed))
  (Utilities|Array|RemoveItem (Variables|Default|GetThrownIceCreams) SelfActor))

(event Custom|PickUpWasDetroyed
  (params (DestroyedActor :type Actor%20Object%20Reference))
  ;[1 1 1] Keep Track if the Item was PickedUp and or Destroyed to make sure the List is up to date.
  (bind _asfort_pickup (Utilities|Casting|CastToFortPickup DestroyedActor))
  (Interaction|UnbindEventfromOnPickup _asfort_pickup (delegate Custom|IceCreamWasPickedUp))
  (Game|UnbindEventfromOnDestroyed _asfort_pickup (delegate Custom|PickUpWasDetroyed))
  (Utilities|Array|RemoveItem (Variables|Default|GetThrownIceCreams) _asfort_pickup))

(fn ConstructionScript
  (|CallParentFunction))
